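# M_TempMaterial_S3 is a prototype translucent, Default Lit material, named as temporary, in the Hera environment folder.
# Its only node is a texture sample of T_CH7S3_TopDown_Blank_roads, which by its name is a top-down roads image.
# The graph shows no connection from that sample to a material output.
# Automatic setting of usage flags in the editor is disabled.
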Material /Hera/Environment/Prototype/BradS/M_TempMaterial_S3
graph {
 properties {bAutomaticallySetUsageInEditor=false, blendMode=BLEND_Translucent, usedShadingModels=Default Lit}
 nodes [TextureSample() {texture=&/T_CH7S3_TopDown_Blank_roads}]
}
